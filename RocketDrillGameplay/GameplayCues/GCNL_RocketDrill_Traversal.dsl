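; Looping gameplay cue driving Rocket Drill traversal VFX, audio and wash force on a Fort pawn.
; On start it listens for jump events and surface changes, feeds Niagara jump, water, ground and water-body inputs, and attaches a physics wash-force actor, gated by effects quality.
; A repeating timer passes linked AnimBP pitch and yaw to audio, with a control bus mix for local viewers.
; Removal animates wash strength, then cleans up force, timer and mix.
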(event Custom|CheckWaterBodyOnTimer
  ;[0.0179 0.5523 0.7954 0.5] WATER BODY
  (Water|SetWaterBodyComponent (Variables|Default|GetFX_Drill_Interaction) "WaterBodyActor" (Variables|/Script/Water.WaterBody|GetWaterBodyComponent (WaterBody|GetCurrentWaterBody (Variables|Default|GetPawnRef)))))

(event EventOnLoopingStart(Niagara)
  (params (MyTarget :type Actor%20Object%20Reference) (Parameters :type Gameplay%20Cue%20Parameters%20Structure%20%28by%20ref%29) (ParticleComponents :type Array%20of%20Niagara%20Particle%20System%20Component%20Object%20References) (AudioComponents :type Array%20of%20Audio%20Component%20Object%20References))
  (bind _pawnref (Variables|Default|GetPawnRef))
  (Variables|Default|SetFX_Drill_Interaction (Utilities|Array|Get(acopy) ParticleComponents))
  (Variables|Audio|SetSpawnedAudioComponents AudioComponents)
  (bind _asfort_pawn (Utilities|Casting|CastToFortPawn MyTarget))
  (bind _output_get (Variables|Default|SetPawnRef _asfort_pawn))
  (Ability|Async|WaitGameplayEventToActor _output_get "(TagName=\"Event.Athena.RocketDrill.Jump.InAir\")"
    (:then
      (Ability|Async|WaitGameplayEventToActor _output_get "(TagName=\"Event.Athena.RocketDrill.Jump.End\")"
        (:then
          (EventDispatchers|BindEventtoOnSurfaceTypeChanged (Footsteps|GetSurfaceTrackingComponent _output_get) (EventDispatchers|CreateEvent :!event Custom|SurfaceTypeChanged))
          (Niagara|SetNiagaraVariable(Bool) (Variables|Default|GetFX_Drill_Interaction) "blsJump" (Variables|Default|GetisJump))
          ;[0 0.5629 0.8876 0.5] WATER BODY
          (bind _returnvalue (Utilities|Time|SetTimerbyEvent (EventDispatchers|CreateEvent :!event Custom|CheckWaterBodyOnTimer) 0.5))
          (Variables|Default|SetWaterTimerHandle _returnvalue)
          (CallFunction|AudioInitialization)
          ;[0.2611 1 0.9568 0.5] WIND FORCE
          (Utilities|EffectsQualityCheck 0
            (:Cinematic
              _)
            (:Epic
              _)
            (:High
              _))
          (Utilities|IsValid (Variables|Default|GetPhysical%20Wash%20Force)
            (:Is%20Not%20Valid
              (bind _returnvalue_2 (Game|SpawnActorfromClass _pawnref "/Game/Effects/Niagara/Blueprints/PhsyicsFields/FN_WashForce.FN_WashForce_C" (Transformation|GetActorTransform self) "Undefined" "MultiplyWithRoot" 0 5e+02 (Variables|Default|GetStrength) -999.0 0))
              (bind _output_get_2 (Variables|Default|SetPhysical%20Wash%20Force _returnvalue_2))
              (Transformation|AttachActorToComponent _output_get_2 (Variables|/Script/Engine.Actor|GetRootComponent _pawnref) "None" "SnapToTarget" "KeepRelative" "KeepRelative" true))) ; Don't spawn the Rotor Wash multiple times
        )
        (:EventReceived
          (Variables|Default|SetisJump false))))
    (:EventReceived
      (Variables|Default|SetisJump true))))

(event Custom|SurfaceTypeChanged
  (params (SurfaceActor :type Actor%20Object%20Reference) (CurrentSurfaceType :type EPhysicalSurface%20Enum :default "SurfaceType_Default") (PrevSurfaceType :type EPhysicalSurface%20Enum :default "SurfaceType_Default"))
  (switch Utilities|FlowControl|Switch|SwitchonEPhysicalSurface CurrentSurfaceType
    (:SurfaceType_Default
      _)
    (:SurfaceType1
      _)
    (:SurfaceType2
      _)
    (:SurfaceType3
      _)
    (:SurfaceType8
      (bind _output_get (Variables|Default|SetIsWater true))
      (bind _output_get_2 (Variables|Default|SetisGrounded false))
      (CallFunction|BlsSurfaceInteractionNiagara _output_get _output_get_2))
    (:SurfaceType9
      _)
    (:SurfaceType10
      _)
    (:SurfaceType18
      _)
    (:SurfaceType23
      _)
    (:SurfaceType24
      _)
    (:SurfaceType25
      _)
    (:SurfaceType26
      _)
    (:SurfaceType30
      _))
  (bind _output_get_3 (Variables|Default|SetIsWater false))
  (bind _output_get_4 (Variables|Default|SetisGrounded true))
  (CallFunction|BlsSurfaceInteractionNiagara _output_get_3 _output_get_4))

(event EventOnRemoval(EitherVFX-preferthisversion)
  (params (MyTarget :type Actor%20Object%20Reference) (Parameters :type Gameplay%20Cue%20Parameters%20Structure%20%28by%20ref%29) (ParticleComponents :type Array%20of%20FXSystem%20Component%20Object%20References) (AudioComponents :type Array%20of%20Audio%20Component%20Object%20References) (BurstCameraShakeInstance :type Camera%20Shake%20Base%20Object%20Reference) (BurstDecalInstance :type Decal%20Actor%20Object%20Reference))
  (|CallParentFunction MyTarget Parameters ParticleComponents AudioComponents BurstCameraShakeInstance BurstDecalInstance)
  (bind (_direction _strength) (Timeline|Timeline Play)
    (:Update
      (Variables|/Game/Effects/Niagara/Blueprints/PhsyicsFields/FN_WashForce.FN_WashForce_C|SetStrength _strength (Variables|Default|GetPhysical%20Wash%20Force)))
    (:Finished
      (Actor|DestroyActor (Variables|Default|GetPhysical%20Wash%20Force))
      (Utilities|Time|ClearandInvalidateTimerbyHandle MyTarget (Variables|Audio|GetFlightMovementTimer))
      (if (Variables|Default|GetisLocallyViewed)
        (Audio|Modulation|DeactivateControlBusMix (Variables|Default|GetPawnRef) (Variables|Default|GetLocalMixModifier))))))

(fn AudioInitialization
  (Variables|Default|SetisLocallyViewed false)
  (bind _asrocket_drill_player_anim_bp (Utilities|Casting|CastToRocketDrill_Player_AnimBP (Components|SkeletalMesh|AnimationBlueprintLinking|GetLinkedAnimLayerInstancebyClass (Variables|/Script/Engine.Character|GetMesh (Variables|Default|GetPawnRef)) "/RocketDrillGameplay/Assets/Animations/RocketDrill_Player_AnimBP.RocketDrill_Player_AnimBP_C")))
  (Variables|Audio|SetRocketDrillAnimBP _asrocket_drill_player_anim_bp)
  (Utilities|Time|ClearandInvalidateTimerbyHandle (Variables|Default|GetPawnRef) (Variables|Audio|GetFlightMovementTimer))
  (bind _returnvalue (Utilities|Time|SetTimerbyEvent (EventDispatchers|CreateEvent :!event MovementTick) 0.1 true true))
  (Variables|Audio|SetFlightMovementTimer _returnvalue)
  (if (and (not (Audio|Modulation|IsControlBusMixActive (Variables|Default|GetPawnRef) (Variables|Default|GetLocalMixModifier))) (Pawn|IsLocallyViewed (Variables|Default|GetPawnRef)))
    (Audio|Modulation|ActivateControlBusMix (Variables|Default|GetPawnRef) (Variables|Default|GetLocalMixModifier))
    (Variables|Default|SetisLocallyViewed true)))

(fn blsSurfaceInteractionNiagara
  (params (blsWater :type Boolean) (blsGround :type Boolean))
  (bind _fx_drill_interaction (Variables|Default|GetFX_Drill_Interaction))
  (Niagara|SetNiagaraVariable(Bool) _fx_drill_interaction "blsWater" blsWater)
  (Niagara|SetNiagaraVariable(Bool) _fx_drill_interaction "blsGround" blsGround))

(fn MovementTick
  (bind _rocketdrillanimbp (Variables|Audio|GetRocketDrillAnimBP)
    (:then
      (LocalVariables|SetPitch (Variables|/Script/RocketDrillGameplayRuntime.RocketDrillLayerAnimInstance|GetVelocityPitch _rocketdrillanimbp))
      (LocalVariables|SetYaw (Variables|/Script/FortniteGame.FortPlayerAnimInstanceProxy|GetAngularVelocityYaw _rocketdrillanimbp))
      (for _array_element (Variables|Audio|GetSpawnedAudioComponents)
        (Utilities|IsValid _array_element
          (:Is%20Valid
            (Audio|Parameter|SetFloatParameter _array_element "Pitch" (LocalVariables|GetPitch) :!class /Script/AudioExtensions.AudioParameterControllerInterface)
            (Audio|Parameter|SetFloatParameter _array_element "Yaw" (LocalVariables|GetYaw) :!class /Script/AudioExtensions.AudioParameterControllerInterface)))))
    (:else)))

(fn ConstructionScript
  (|CallParentFunction))
