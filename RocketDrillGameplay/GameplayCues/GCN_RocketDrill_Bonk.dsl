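; Gameplay cue notify for the Rocket Drill "bonk" that relays a hard impact to the cue's target actor as a gameplay event.
; On execute, it sends Event.Athena.RocketDrill.HardImpact to the target, packing the cue's hit result into the event's target data, then calls the parent implementation.
; It also debug-prints whether the cue has a hit result, but sends the event without checking that.
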(fn OnExecute
  (params (MyTarget :type Actor%20Object%20Reference) (Parameters :type Gameplay%20Cue%20Parameters%20Structure%20%28by%20ref%29))
  (Development|PrintString MyTarget (Utilities|String|ToString(Boolean) (Ability|GameplayCue|HasHitResult Parameters)))
  (Ability|SendGameplayEventtoActor MyTarget "(TagName=\"Event.Athena.RocketDrill.HardImpact\")" (Utilities|Struct|MakeGameplayEventData "(TagName=\"Event.Athena.RocketDrill.HardImpact\")" 0 0 0 0 "()" "(GameplayTags=)" "(GameplayTags=)" 0.0 (Ability|TargetData|AbilityTargetDatafromHitResult (Ability|GameplayCue|GetHitResult Parameters))))
  (return (|CallParentFunction MyTarget Parameters)))
