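; Parent burst GameplayCue notify for Rocket Drill effects, providing shared local-feedback helpers.
; OnBurst passes spawned audio components and cue location to SpawnSoundIndicator, then calls the parent implementation.
; The indicator uses the first sound's max distance and shows a gunshot marker for neutral and hostile affiliations relative to the target's team.
; Camera shake and lens effect helpers, uncalled here, play only for locally controlled pawns.
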(fn OnBurst
  (params (Target :type Actor%20Object%20Reference) (Parameters :type Gameplay%20Cue%20Parameters%20Structure%20%28by%20ref%29) (SpawnResults :type Gameplay%20Cue%20Notify%20Spawn%20Result%20Structure%20%28by%20ref%29))
  (bind (_normalizedmagnitude _rawmagnitude _effectcontext _matchedtagname _originaltag _aggregatedsourcetags _aggregatedtargettags _location) (Ability|GameplayCue|BreakGameplayCueParameters Parameters))
  (LocalVariables|SetSpawnedComponents (Utilities|Struct|BreakGameplayCueNotifySpawnResult SpawnResults))
  (Default|SpawnSoundIndicator (Variables|GetTarget) (LocalVariables|GetSpawnedComponents) _location)
  (|CallParentFunction Target Parameters SpawnResults)
  (return))

(fn PlayLocalCameraLensEffect
  (params (FortPlayerPawn :type Fort%20Player%20Pawn%20Object%20Reference) (LensEffectEmitterClass :type Actor%20Class%20Reference))
  ; Play Camera Lens Effect if the player pawn is local.
  (Utilities|IsValid FortPlayerPawn
    (:Is%20Valid
      (bind _asplayer_controller (Utilities|Casting|CastToPlayerController (Pawn|GetController FortPlayerPawn)))
      (if (Pawn|IsLocalController _asplayer_controller)
        (Game|Feedback|ClientSpawnGenericCameraLensEffect _asplayer_controller LensEffectEmitterClass)))))

(fn PlayLocalCameraShake
  (params (FortPlayerPawn :type Fort%20Player%20Pawn%20Object%20Reference) (ShakeClass :type Camera%20Shake%20Base%20Class%20Reference))
  ; Play Camera Shake if the player pawn is local.
  (Utilities|IsValid FortPlayerPawn
    (:Is%20Valid
      (bind _asplayer_controller (Utilities|Casting|CastToPlayerController (Pawn|GetController FortPlayerPawn)))
      (if (Pawn|IsLocalController _asplayer_controller)
        (bind _playercameramanager (Variables|/Script/Engine.PlayerController|GetPlayerCameraManager _asplayer_controller)
          (:then
            (CameraShakes|StartCameraShake :self _playercameramanager :ShakeClass ShakeClass :Scale 1.0 :PlaySpace "CameraLocal"))
          (:else))))))

(fn SpawnSoundIndicator
  (params (TargetActor :type Actor%20Object%20Reference) (AudioComponents :type Array%20of%20Audio%20Component%20Object%20References) (Location :type Vector))
  (bind _output (Utilities|Array|Get(acopy) AudioComponents))
  ;[0.02051 0.132 0.15 0.5] SpawnSoundIndicator: Spawn the sound indicator, if desired and there's a valid sound played.
  (if (Utilities|Array|IsValidIndex AudioComponents)
    (Utilities|IsValid _output
      (:Is%20Valid
        (bind _sound (Variables|/Script/Engine.AudioComponent|GetSound _output)
          (:then
            (Fortnite|ShowSoundIndicatorLocalforTeam :WorldContextObject TargetActor :Location Location :MaxAudibleDistance (Variables|/Script/Engine.SoundBase|GetMaxDistance _sound) :IndicatorType "Gunshot" :Team (Team|GetActorTeam TargetActor) :Affiliations (Utilities|Array|MakeArray "Neutral" "Hostile") :Tint "(R=0.000000,G=0.000000,B=0.000000,A=0.000000)" :AttenuationMaxRangeMultiplier 1.0 :AttenuationIntensityMultiplier 1.0))
          (:else))))))
